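; Controller requirement for the ExpertSale conversation service, named as a token threshold check, that decides whether a requesting controller meets the requirement.
; The IsRequirementMetInternal override returns true immediately when bSkip is set, bypassing the check.
; Otherwise it runs the parent implementation on the request context and returns that result, negated when bInvertResult is set.
; The threshold comparison itself is not in this graph, which only adds the skip and invert filtering.
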(fn IsRequirementMetInternal
  (params (RequestContext :type Controller%20Requirement%20Test%20Context%20Structure%20%28by%20ref%29))
  (if (Variables|Filtering|GetbSkip)
    (return true)
    (else
      (bind _returnvalue (|CallParentFunction RequestContext))
      (if (Variables|Filtering|GetbInvertResult)
        (return (not _returnvalue))
        (else
          (return _returnvalue))))))
